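# Physics asset for the F_MED_SmartHyena female player body, covering dangling ribbon bones and the lower body around them.
# Four ribbon chains (A to D), each three to five capsules long, link bone to bone and hang from the pelvis.
# Larger capsules on the pelvis and both thighs form the lower-body volumes, with each thigh constrained to the pelvis.
# Every constraint carries the same swing and twist limit angles.
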PhysicsAsset /Game/Characters/Player/Female/Medium/Bodies/F_MED_SmartHyena/Meshes/F_MED_SmartHyena_Physics
bodies {
 dyn_ribbon_A01 [{
   shapeType=Capsule
   length=8.2493
   radius=0.7965
   shapeName=dyn_ribbon_A01_capsule
   center=<-4.3276 -1.7031 0.135>
   rotation=<rot 0 -71.3476 -89.7073>
  }]
 dyn_ribbon_A02 [{
   shapeType=Capsule
   length=8.5193
   radius=0.8548
   shapeName=dyn_ribbon_A02_capsule
   center=<-4.4662 -0.8518 0.0868>
   rotation=<rot 0 -86.8847 -89.6791>
  }]
 dyn_ribbon_A03 [{
   shapeType=Capsule
   length=8.6693
   radius=0.9009
   shapeName=dyn_ribbon_A03_capsule
   center=<-5.5669 -0.1395 0.1556>
   rotation=<rot 0 -87.3843 -87.4368>
  }]
 dyn_ribbon_A04 [{
   shapeType=Capsule
   length=8.7895
   radius=0.9256
   shapeName=dyn_ribbon_A04_capsule
   center=<-4.672 -0.0846 -0.0812>
   rotation=<rot 0 -87.9763 -89.5144>
  }]
 dyn_ribbon_A05 [{
   shapeType=Capsule
   length=8.1456
   radius=1.0289
   shapeName=dyn_ribbon_A05_capsule
   center=<-5.3838 -0.4635 0.0854>
   rotation=<rot 0 -88.2333 -89.7571>
  }]
 dyn_ribbon_B01 [{
   shapeType=Capsule
   length=7.7714
   radius=0.9352
   shapeName=dyn_ribbon_B01_capsule
   center=<-4.6607 -1.7343 0.1449>
   rotation=<rot 0 -71.4991 -85.3611>
  }]
 dyn_ribbon_B02 [{
   shapeType=Capsule
   length=8.71
   radius=0.8821
   shapeName=dyn_ribbon_B02_capsule
   center=<-4.2318 -0.7791 0.2851>
   rotation=<rot 0 -82.9012 -86.6231>
  }]
 dyn_ribbon_B03 [{
   shapeType=Capsule
   length=9.1964
   radius=1.005
   shapeName=dyn_ribbon_B03_capsule
   center=<-4.802 -1.0532 -0.0622>
   rotation=<rot 0 97.5146 -87.0717>
  }]
 dyn_ribbon_C01 [{
   shapeType=Capsule
   length=8.446
   radius=0.9699
   shapeName=dyn_ribbon_C01_capsule
   center=<-3.8094 -2.0319 -0.272>
   rotation=<rot 0 -66.6687 -88.8189>
  }]
 dyn_ribbon_C02 [{
   shapeType=Capsule
   length=8.6292
   radius=0.7723
   shapeName=dyn_ribbon_C02_capsule
   center=<-3.841 -0.8265 -0.0266>
   rotation=<rot 0 91.7161 -87.6171>
  }]
 dyn_ribbon_C03 [{
   shapeType=Capsule
   length=6.005
   radius=0.505
   shapeName=dyn_ribbon_C03_capsule
   center=<-3.6852 0.0548 -0.2024>
   rotation=<rot 0 93.1375 -85.1957>
  }]
 dyn_ribbon_C04 [{
   shapeType=Capsule
   length=7.4219
   radius=0.8413
   shapeName=dyn_ribbon_C04_capsule
   center=<-2.5414 -0.3317 -0.4999>
   rotation=<rot 0 95.0562 -81.122>
  }]
 dyn_ribbon_D01 [{
   shapeType=Capsule
   length=7.7441
   radius=0.8557
   shapeName=dyn_ribbon_D01_capsule
   center=<-4.7581 -1.4576 -0.6234>
   rotation=<rot 0 113.0643 -83.7823>
  }]
 dyn_ribbon_D02 [{
   shapeType=Capsule
   length=6.005
   radius=0.505
   shapeName=dyn_ribbon_D02_capsule
   center=<-3.8869 -0.858 -0.1372>
   rotation=<rot 0 104.4693 -86.9237>
  }]
 dyn_ribbon_D03 [{
   shapeType=Capsule
   length=6.005
   radius=0.505
   shapeName=dyn_ribbon_D03_capsule
   center=<-4.9061 -0.4556 -0.2538>
   rotation=<rot 0 95.7494 -86.1127>
  }]
 dyn_ribbon_D04 [{
   shapeType=Capsule
   length=6.005
   radius=0.505
   shapeName=dyn_ribbon_D04_capsule
   center=<-3.6426 -0.3504 -0.1628>
   rotation=<rot 0 95.7266 -86.4591>
  }]
 dyn_ribbon_D05 [{
   shapeType=Capsule
   length=7.5294
   radius=1.0544
   shapeName=dyn_ribbon_D05_capsule
   center=<-2.7223 -0.1604 -0.3283>
   rotation=<rot 0 94.5995 -84.3763>
  }]
 pelvis [{shapeType=Capsule, length=16.5, radius=13.5, shapeName=pelvis_capsule, center=<-4.7102 0 0>}]
 thigh_l [{shapeType=Capsule, length=19.2837, radius=14.6117, center=<-25.2688 -0.5634 0.2188>, rotation=<rot -0.2204 -86.5477 -91.2501>}]
 thigh_r [{shapeType=Capsule, length=22.1457, radius=12.6512, center=<26.1696 2.1915 -0.0641>, rotation=<rot 1.6683 -88.67 -85.4103>}]
}
constraints [
 {bone1Name=thigh_l, bone2Name=pelvis, swing1LimitDegrees=45, swing2LimitDegrees=45, twistLimitDegrees=45}
 {bone1Name=thigh_r, bone2Name=pelvis, swing1LimitDegrees=45, swing2LimitDegrees=45, twistLimitDegrees=45}
 {bone1Name=dyn_ribbon_A01, bone2Name=pelvis, swing1LimitDegrees=45, swing2LimitDegrees=45, twistLimitDegrees=45}
 {bone1Name=dyn_ribbon_A02, bone2Name=dyn_ribbon_A01, swing1LimitDegrees=45, swing2LimitDegrees=45, twistLimitDegrees=45}
 {bone1Name=dyn_ribbon_A03, bone2Name=dyn_ribbon_A02, swing1LimitDegrees=45, swing2LimitDegrees=45, twistLimitDegrees=45}
 {bone1Name=dyn_ribbon_A04, bone2Name=dyn_ribbon_A03, swing1LimitDegrees=45, swing2LimitDegrees=45, twistLimitDegrees=45}
 {bone1Name=dyn_ribbon_A05, bone2Name=dyn_ribbon_A04, swing1LimitDegrees=45, swing2LimitDegrees=45, twistLimitDegrees=45}
 {bone1Name=dyn_ribbon_B01, bone2Name=pelvis, swing1LimitDegrees=45, swing2LimitDegrees=45, twistLimitDegrees=45}
 {bone1Name=dyn_ribbon_B02, bone2Name=dyn_ribbon_B01, swing1LimitDegrees=45, swing2LimitDegrees=45, twistLimitDegrees=45}
 {bone1Name=dyn_ribbon_B03, bone2Name=dyn_ribbon_B02, swing1LimitDegrees=45, swing2LimitDegrees=45, twistLimitDegrees=45}
 {bone1Name=dyn_ribbon_C01, bone2Name=pelvis, swing1LimitDegrees=45, swing2LimitDegrees=45, twistLimitDegrees=45}
 {bone1Name=dyn_ribbon_C02, bone2Name=dyn_ribbon_C01, swing1LimitDegrees=45, swing2LimitDegrees=45, twistLimitDegrees=45}
 {bone1Name=dyn_ribbon_C03, bone2Name=dyn_ribbon_C02, swing1LimitDegrees=45, swing2LimitDegrees=45, twistLimitDegrees=45}
 {bone1Name=dyn_ribbon_C04, bone2Name=dyn_ribbon_C03, swing1LimitDegrees=45, swing2LimitDegrees=45, twistLimitDegrees=45}
 {bone1Name=dyn_ribbon_D01, bone2Name=pelvis, swing1LimitDegrees=45, swing2LimitDegrees=45, twistLimitDegrees=45}
 {bone1Name=dyn_ribbon_D02, bone2Name=dyn_ribbon_D01, swing1LimitDegrees=45, swing2LimitDegrees=45, twistLimitDegrees=45}
 {bone1Name=dyn_ribbon_D03, bone2Name=dyn_ribbon_D02, swing1LimitDegrees=45, swing2LimitDegrees=45, twistLimitDegrees=45}
 {bone1Name=dyn_ribbon_D04, bone2Name=dyn_ribbon_D03, swing1LimitDegrees=45, swing2LimitDegrees=45, twistLimitDegrees=45}
 {bone1Name=dyn_ribbon_D05, bone2Name=dyn_ribbon_D04, swing1LimitDegrees=45, swing2LimitDegrees=45, twistLimitDegrees=45}
]
