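# Post-process style AnimBlueprint for the LongBandit female body, using SmartHyena rig assets to animate skirt, ribbons, jewelry and attach points.
# It copies the parent pose, applies bone constraints and Control Rigs, then caches it.
# Optional AnimDynamics jewelry simulation blends with that cache, weighted by DBNO state, before a final RigidBody pass.
# The event graph updates cape gravity, WPO cloth and wind material parameters, and cape-dependent scarf visibility.
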AnimBlueprint /BRCosmetics/Characters/Player/Female/Medium/Bodies/F_MED_LongBandit/Meshes/F_MED_SmartHyena_AnimBP
target_skeleton=/Game/Characters/Player/Female/Medium/Bodies/F_MED_SmartHyena/Meshes/F_MED_SmartHyena_Skeleton
graphs [
 AnimGraph {nodes [
   n1 = Get bIsDBNO()
   n2 = Blend Poses by bool(
    BlendPose_0=Two Way Blend(A=Use cached pose 'BodyDynPose'(), B=Use cached pose 'BodyCachePose'(), Alpha=SelectFloat(A=0.95, B=0.2, bPickA=@n1)) {blendNode {alphaBoolBlend {blendOption=Linear}}}
    BlendPose_1=Use cached pose 'BodyDynPose'()
    bActiveValue=Get bIsDBNO()
   ) {blendType=Linear}
   n3 = Use cached pose 'BodyCachePose'()
   n4 = Get bAnimDynamicsEnabled() -> (bAnimDynamicsEnabled:bool)
   n5 = Get isWearingPackBool() -> (isWearingPackBool:bool)
   n6 = Get bIsSkyDiving()
   n7 = RigidBody!(
    ComponentPose=Local To Component(LocalPose=Blend Poses by bool(BlendPose_0=@n2, BlendPose_1=@n3, bActiveValue=@n4) {blendType=Linear})
    Alpha=SelectFloat(A=0.8, B=1, bPickA=Get isWearingPackBool())
    OverrideWorldGravity:Vector=SelectVector(A="0,0,-980", B=Get Gravity Override Cape() -> (Gravity Override Cape:Vector), bPickA=AND Boolean(A=@n6, B=@n5))
    ExternalForce
    ComponentLinearAccScale:Vector="0.5,0.2,0.4"
    ComponentLinearVelScale:Vector="0.8,0.35,0.45"
   ) -> (Pose) {
    bClampLinearTranslationLimitToRefPose=true
    bOverrideWorldGravity=true
    overridePhysicsAsset=/Game/Characters/Player/Female/Medium/Bodies/F_MED_SmartHyena/Meshes/F_MED_SmartHyena_Physics
    simulationSpace=BaseBoneSpace
    alphaBoolBlend {blendOption=Linear}
    baseBoneRef {boneName=gravity}
    componentAppliedLinearAccClamp=<3500 3500 3500>
    componentLinearAccScale=<0.1 0.1 0.1>
    componentLinearVelScale=<0.1 0.1 0.1>
    overrideWorldGravity {z=-980}
    simSpaceSettings {
     maxAngularAcceleration=1000
     maxAngularVelocity=0.3
     maxLinearAcceleration=700
     maxLinearVelocity=500
     velocityScaleZ=0
     worldAlpha=0.1
     externalLinearDragV=<0.5 0.5>
    }
   }
   Output Pose(Result=Component To Local(ComponentPose=@n7)) {layerGroup=None}
   n8 = "Constraint - attach_backpack (PPPP)"!(
    ComponentPose=Local To Component(LocalPose=Copy Pose From Mesh() {bUseAttachedParent=true})
    Alpha=1
    ConstraintWeights_0=0.6869
    ConstraintWeights_1:float=0.1785
    ConstraintWeights_2:float=0.12
    ConstraintWeights_3:float=0.0146
   ) -> (Pose) {
    alphaBoolBlend {blendOption=Linear}
    boneToModify {boneName=attach_backpack}
    constraintSetup [{targetBone {boneName=spine_05}}, {targetBone {boneName=spine_04}}, {targetBone {boneName=clavicle_r}}, {targetBone {boneName=clavicle_l}}]
    constraintWeights [1,1,1,1]
   }
   n9 = "Constraint - attach_cape (P)"!(ComponentPose=@n8, Alpha=1, ConstraintWeights_0=1, ConstraintWeights_1:float=0) -> (Pose) {
    alphaBoolBlend {blendOption=Linear}
    boneToModify {boneName=attach_cape}
    constraintSetup [{targetBone {boneName=spine_05}}]
    constraintWeights [1,0]
   }
   n10 = Control Rig(Source=Component To Local(ComponentPose=@n9)) {
    alphaBoolBlend {blendOption=Linear}
    controlRigAssetReference {blueprintRigClass=/Game/Characters/Player/Female/Medium/Bodies/F_MED_SmartHyena/Meshes/F_MED_SmartHyena_ControlRig_new.F_MED_SmartHyena_ControlRig_new_C}
   }
   n11 = Control Rig(Source=@n10) {
    alphaBoolBlend {blendOption=Linear}
    controlRigAssetReference {blueprintRigClass=/Game/Characters/Player/Female/Medium/Bodies/F_MED_SmartHyena/Meshes/F_MED_SmartHyena_Skirt_CtrlRig.F_MED_SmartHyena_Skirt_CtrlRig_C}
   }
   n12 = "Constraint - gravity (TT)"!(ComponentPose=Local To Component(LocalPose=@n11), Alpha=1, ConstraintWeights_0=0.3, ConstraintWeights_1:float=0.7) -> (Pose) {
    alphaBoolBlend {blendOption=Linear}
    boneToModify {boneName=gravity}
    constraintSetup [{transformType=Translation, targetBone {boneName=root}}, {transformType=Translation, targetBone {boneName=pelvis}}]
    constraintWeights [1,1]
   }
   n13 = "Constraint - dyn_ribbon_A03 (PP)"!(ComponentPose=@n12, Alpha=1, ConstraintWeights_0=0.5, ConstraintWeights_1:float=0.5) -> (Pose) {
    alphaBoolBlend {blendOption=Linear}
    boneToModify {boneName=dyn_ribbon_A03}
    constraintSetup [{targetBone {boneName=pelvis}}, {targetBone {boneName=thigh_twist_01_r}}]
    constraintWeights [1,1]
   }
   n14 = "Constraint - dyn_ribbon_D03 (PP)"!(ComponentPose=@n13, Alpha=1, ConstraintWeights_0=0.5, ConstraintWeights_1:float=0.5) -> (Pose) {
    alphaBoolBlend {blendOption=Linear}
    boneToModify {boneName=dyn_ribbon_D03}
    constraintSetup [{targetBone {boneName=pelvis}}, {targetBone {boneName=thigh_twist_01_l}}]
    constraintWeights [1,1]
   }
   n15 = "Constraint - dyn_jewelry_A01 (PPPP)"!(
    ComponentPose=@n14
    Alpha=1
    ConstraintWeights_0=0.7626
    ConstraintWeights_1:float=0.1963
    ConstraintWeights_2:float=0.0387
    ConstraintWeights_3:float=0.0024
   ) -> (Pose) {
    alphaBoolBlend {blendOption=Linear}
    boneToModify {boneName=dyn_jewelry_A01}
    constraintSetup [
     {targetBone {boneName=pelvis}}
     {targetBone {boneName=dyn_Cloth_volume_r_A}}
     {targetBone {boneName=spine_01}}
     {targetBone {boneName=dyn_Cloth_volume_l_A}}
    ]
    constraintWeights [1,1,1,1]
   }
   n16 = "Constraint - dyn_jewelry_A02 (PPPPPPP)"!(
    ComponentPose=@n15
    Alpha=0.5
    ConstraintWeights_0=0.6309
    ConstraintWeights_1:float=0.3074
    ConstraintWeights_2:float=0.0331
    ConstraintWeights_3:float=0.0147
    ConstraintWeights_4:float=0.0107
    ConstraintWeights_5:float=0.0021
    ConstraintWeights_6:float=0.0011
   ) -> (Pose) {
    alphaBoolBlend {blendOption=Linear}
    boneToModify {boneName=dyn_jewelry_A02}
    constraintSetup [
     {targetBone {boneName=pelvis}}
     {targetBone {boneName=dyn_Cloth_volume_r_A}}
     {targetBone {boneName=thigh_twist_01_r}}
     {targetBone {boneName=dyn_Cloth_volume_l_A}}
     {targetBone {boneName=thigh_r}}
     {targetBone {boneName=thigh_twist_01_l}}
     {targetBone {boneName=dyn_Cloth_volume_fn}}
    ]
    constraintWeights [1,1,1,1,1,1,1]
   }
   n17 = "Constraint - dyn_jewelry_A03 (PPPPPPPP)"!(
    ComponentPose=@n16
    Alpha=0.5
    ConstraintWeights_0=0.3924
    ConstraintWeights_1:float=0.2812
    ConstraintWeights_2:float=0.1378
    ConstraintWeights_3:float=0.0695
    ConstraintWeights_4:float=0.055
    ConstraintWeights_5:float=0.0331
    ConstraintWeights_6:float=0.0209
    ConstraintWeights_7:float=0.0101
   ) -> (Pose) {
    alphaBoolBlend {blendOption=Linear}
    boneToModify {boneName=dyn_jewelry_A03}
    constraintSetup [
     {targetBone {boneName=pelvis}}
     {targetBone {boneName=dyn_Cloth_volume_r_A}}
     {targetBone {boneName=thigh_twist_01_r}}
     {targetBone {boneName=thigh_r}}
     {targetBone {boneName=dyn_Cloth_volume_fn}}
     {targetBone {boneName=thigh_twist_01_l}}
     {targetBone {boneName=dyn_Cloth_volume_l_A}}
     {targetBone {boneName=thigh_l}}
    ]
    constraintWeights [1,1,1,1,1,1,1,1]
   }
   n18 = "Constraint - dyn_jewelry_B01 (PPPPPP)"!(
    ComponentPose=@n17
    Alpha=1
    ConstraintWeights_0=0.6704
    ConstraintWeights_1:float=0.2848
    ConstraintWeights_2:float=0.0366
    ConstraintWeights_3:float=0.0047
    ConstraintWeights_4:float=0.0033
    ConstraintWeights_5:float=0.0001
   ) -> (Pose) {
    alphaBoolBlend {blendOption=Linear}
    boneToModify {boneName=dyn_jewelry_B01}
    constraintSetup [
     {targetBone {boneName=pelvis}}
     {targetBone {boneName=spine_01}}
     {targetBone {boneName=spine_02}}
     {targetBone {boneName=dyn_Cloth_volume_r_A}}
     {targetBone {boneName=spine_03}}
     {targetBone {boneName=spine_04}}
    ]
    constraintWeights [1,1,1,1,1,1]
   }
   n19 = "Constraint - dyn_jewelry_B02 (PPPPPP)"!(
    ComponentPose=@n18
    Alpha=1
    ConstraintWeights_0=0.6924
    ConstraintWeights_1:float=0.2554
    ConstraintWeights_2:float=0.0509
    ConstraintWeights_3:float=0.0006
    ConstraintWeights_4:float=0.0005
    ConstraintWeights_5:float=0.0002
   ) -> (Pose) {
    alphaBoolBlend {blendOption=Linear}
    boneToModify {boneName=dyn_jewelry_B02}
    constraintSetup [
     {targetBone {boneName=pelvis}}
     {targetBone {boneName=dyn_Cloth_volume_r_A}}
     {targetBone {boneName=spine_01}}
     {targetBone {boneName=spine_03}}
     {targetBone {boneName=dyn_Cloth_volume_l_A}}
     {targetBone {boneName=spine_02}}
    ]
    constraintWeights [1,1,1,1,1,1]
   }
   n20 = "Constraint - dyn_jewelry_B03 (PPPPPP)"!(
    ComponentPose=@n19
    Alpha=1
    ConstraintWeights_0=0.5024
    ConstraintWeights_1:float=0.4059
    ConstraintWeights_2:float=0.0693
    ConstraintWeights_3:float=0.0147
    ConstraintWeights_4:float=0.0068
    ConstraintWeights_5:float=0.0009
   ) -> (Pose) {
    alphaBoolBlend {blendOption=Linear}
    boneToModify {boneName=dyn_jewelry_B03}
    constraintSetup [
     {targetBone {boneName=dyn_Cloth_volume_r_A}}
     {targetBone {boneName=pelvis}}
     {targetBone {boneName=thigh_twist_01_r}}
     {targetBone {boneName=thigh_r}}
     {targetBone {boneName=dyn_Cloth_volume_fn}}
     {targetBone {boneName=thigh_twist_01_l}}
    ]
    constraintWeights [1,1,1,1,1,1]
   }
   n21 = "Constraint - dyn_jewelry_B04 (PPPPP)"!(
    ComponentPose=@n20
    Alpha=1
    ConstraintWeights_0=0.442
    ConstraintWeights_1:float=0.2454
    ConstraintWeights_2:float=0.1815
    ConstraintWeights_3:float=0.1241
    ConstraintWeights_4:float=0.007
   ) -> (Pose) {
    alphaBoolBlend {blendOption=Linear}
    boneToModify {boneName=dyn_jewelry_B04}
    constraintSetup [
     {targetBone {boneName=dyn_Cloth_volume_r_A}}
     {targetBone {boneName=pelvis}}
     {targetBone {boneName=thigh_twist_01_r}}
     {targetBone {boneName=thigh_r}}
     {targetBone {boneName=dyn_Cloth_volume_fn}}
    ]
    constraintWeights [1,1,1,1,1]
   }
   n22 = "Constraint - dyn_jewelry_B05 (PPPPP)"!(
    ComponentPose=@n21
    Alpha=1
    ConstraintWeights_0=0.3138
    ConstraintWeights_1:float=0.2438
    ConstraintWeights_2:float=0.223
    ConstraintWeights_3:float=0.2176
    ConstraintWeights_4:float=0.0019
   ) -> (Pose) {
    alphaBoolBlend {blendOption=Linear}
    boneToModify {boneName=dyn_jewelry_B05}
    constraintSetup [
     {targetBone {boneName=dyn_Cloth_volume_r_A}}
     {targetBone {boneName=thigh_twist_01_r}}
     {targetBone {boneName=thigh_r}}
     {targetBone {boneName=pelvis}}
     {targetBone {boneName=dyn_Cloth_volume_fn}}
    ]
    constraintWeights [1,1,1,1,1]
   }
   n23 = "Constraint - dyn_jewelry_Ba01 (PPPPP)"!(
    ComponentPose=@n22
    Alpha=1
    ConstraintWeights_0=0.7791
    ConstraintWeights_1:float=0.2044
    ConstraintWeights_2:float=0.0158
    ConstraintWeights_3:float=0.0005
    ConstraintWeights_4:float=0.0002
   ) -> (Pose) {
    alphaBoolBlend {blendOption=Linear}
    boneToModify {boneName=dyn_jewelry_Ba01}
    constraintSetup [
     {targetBone {boneName=pelvis}}
     {targetBone {boneName=dyn_Cloth_volume_r_A}}
     {targetBone {boneName=thigh_twist_01_r}}
     {targetBone {boneName=spine_01}}
     {targetBone {boneName=thigh_r}}
    ]
    constraintWeights [1,1,1,1,1]
   }
   n24 = "Constraint - dyn_jewelry_Ba02 (PPPPP)"!(
    ComponentPose=@n23
    Alpha=1
    ConstraintWeights_0=0.6147
    ConstraintWeights_1:float=0.2926
    ConstraintWeights_2:float=0.0703
    ConstraintWeights_3:float=0.0216
    ConstraintWeights_4:float=0.0008
   ) -> (Pose) {
    alphaBoolBlend {blendOption=Linear}
    boneToModify {boneName=dyn_jewelry_Ba02}
    constraintSetup [
     {targetBone {boneName=pelvis}}
     {targetBone {boneName=dyn_Cloth_volume_r_A}}
     {targetBone {boneName=thigh_twist_01_r}}
     {targetBone {boneName=thigh_r}}
     {targetBone {boneName=index_01_r}}
    ]
    constraintWeights [1,1,1,1,1]
   }
   n25 = "Constraint - dyn_jewelry_Ba03 (PPPP)"!(
    ComponentPose=@n24
    Alpha=1
    ConstraintWeights_0=0.319
    ConstraintWeights_1:float=0.3154
    ConstraintWeights_2:float=0.2059
    ConstraintWeights_3:float=0.1598
   ) -> (Pose) {
    alphaBoolBlend {blendOption=Linear}
    boneToModify {boneName=dyn_jewelry_Ba03}
    constraintSetup [
     {targetBone {boneName=dyn_Cloth_volume_r_A}}
     {targetBone {boneName=pelvis}}
     {targetBone {boneName=thigh_twist_01_r}}
     {targetBone {boneName=thigh_r}}
    ]
    constraintWeights [1,1,1,1]
   }
   n26 = "Constraint - dyn_jewelry_C01 (PPPPP)"!(
    ComponentPose=@n25
    Alpha=1
    ConstraintWeights_0=0.9086
    ConstraintWeights_1:float=0.0479
    ConstraintWeights_2:float=0.029
    ConstraintWeights_3:float=0.0142
    ConstraintWeights_4:float=0.0003
   ) -> (Pose) {
    alphaBoolBlend {blendOption=Linear}
    boneToModify {boneName=dyn_jewelry_C01}
    constraintSetup [
     {targetBone {boneName=pelvis}}
     {targetBone {boneName=dyn_Cloth_volume_r_A}}
     {targetBone {boneName=thigh_twist_01_r}}
     {targetBone {boneName=spine_01}}
     {targetBone {boneName=spine_02}}
    ]
    constraintWeights [1,1,1,1,1]
   }
   n27 = "Constraint - dyn_jewelry_C02 (PPPP)"!(
    ComponentPose=@n26
    Alpha=1
    ConstraintWeights_0=0.7752
    ConstraintWeights_1:float=0.1768
    ConstraintWeights_2:float=0.043
    ConstraintWeights_3:float=0.005
   ) -> (Pose) {
    alphaBoolBlend {blendOption=Linear}
    boneToModify {boneName=dyn_jewelry_C02}
    constraintSetup [
     {targetBone {boneName=pelvis}}
     {targetBone {boneName=dyn_Cloth_volume_r_A}}
     {targetBone {boneName=thigh_twist_01_r}}
     {targetBone {boneName=thigh_r}}
    ]
    constraintWeights [1,1,1,1]
   }
   n28 = "Transform (Modify) Bone - Bone: dyn_jewelry_B05"(ComponentPose=@n27, Alpha=1, Translation="0,0,0", Rotation="6,0,0", Scale="1,1,1") {
    rotationMode=BMM_Additive
    rotationSpace=BCS_BoneSpace
    scaleMode=BMM_Additive
    scaleSpace=BCS_BoneSpace
    translationMode=BMM_Additive
    translationSpace=BCS_BoneSpace
    alphaBoolBlend {blendOption=Linear}
    boneToModify {boneName=dyn_jewelry_B05}
    rotation {pitch=6}
   }
   n29 = Control Rig(Source=Component To Local(ComponentPose=@n28)) {
    alphaBoolBlend {blendOption=Linear}
    controlRigAssetReference {blueprintRigClass=/Game/Characters/Player/Female/Medium/Bodies/F_MED_SmartHyena/Meshes/F_MED_SmartHyena_jewelry_CtrlRig.F_MED_SmartHyena_jewelry_CtrlRig_C}
   }
   Save cached pose 'BodyCachePose'(Pose=@n29) {cacheName=BodyCachePose}
   n30 = "Transform (Modify) Bone - Bone: dyn_jewelry_A01"(ComponentPose=Local To Component(LocalPose=Use cached pose 'BodyCachePose'()), Alpha=1, Translation="0,0,0", Rotation="5,0,0", Scale="1,1,1") {
    rotationMode=BMM_Additive
    rotationSpace=BCS_BoneSpace
    scaleMode=BMM_Additive
    scaleSpace=BCS_BoneSpace
    translationMode=BMM_Additive
    translationSpace=BCS_BoneSpace
    alphaBoolBlend {blendOption=Linear}
    boneToModify {boneName=dyn_jewelry_A01}
    rotation {pitch=5}
   }
   n31 = "Transform (Modify) Bone - Bone: dyn_jewelry_A02"(ComponentPose=@n30, Alpha=1, Translation="0,0,0", Rotation="5,0,0", Scale="1,1,1") {
    rotationMode=BMM_Additive
    rotationSpace=BCS_BoneSpace
    scaleMode=BMM_Additive
    scaleSpace=BCS_BoneSpace
    translationMode=BMM_Additive
    translationSpace=BCS_BoneSpace
    alphaBoolBlend {blendOption=Linear}
    boneToModify {boneName=dyn_jewelry_A02}
    rotation {pitch=5}
   }
   n32 = "Transform (Modify) Bone - Bone: dyn_jewelry_A03"(ComponentPose=@n31, Alpha=1, Translation="0,0,0", Rotation="5,0,0", Scale="1,1,1") {
    rotationMode=BMM_Additive
    rotationSpace=BCS_BoneSpace
    scaleMode=BMM_Additive
    scaleSpace=BCS_BoneSpace
    translationMode=BMM_Additive
    translationSpace=BCS_BoneSpace
    alphaBoolBlend {blendOption=Linear}
    boneToModify {boneName=dyn_jewelry_A03}
    rotation {pitch=5}
   }
   n33 = "Transform (Modify) Bone - Bone: dyn_jewelry_B01"(ComponentPose=@n32, Alpha=1, Translation="0,0,0", Rotation="0,0,0", Scale="1,1,1") {
    rotationMode=BMM_Additive
    rotationSpace=BCS_BoneSpace
    scaleMode=BMM_Additive
    scaleSpace=BCS_BoneSpace
    translationMode=BMM_Additive
    translationSpace=BCS_BoneSpace
    alphaBoolBlend {blendOption=Linear}
    boneToModify {boneName=dyn_jewelry_B01}
   }
   n34 = "Transform (Modify) Bone - Bone: dyn_jewelry_B02"(ComponentPose=@n33, Alpha=1, Translation="0,0,0", Rotation="0,0,0", Scale="1,1,1") {
    rotationMode=BMM_Additive
    rotationSpace=BCS_BoneSpace
    scaleMode=BMM_Additive
    scaleSpace=BCS_BoneSpace
    translationMode=BMM_Additive
    translationSpace=BCS_BoneSpace
    alphaBoolBlend {blendOption=Linear}
    boneToModify {boneName=dyn_jewelry_B02}
   }
   n35 = "Transform (Modify) Bone - Bone: dyn_jewelry_B03"(ComponentPose=@n34, Alpha=1, Translation="0,0,0", Rotation="0,0,0", Scale="1,1,1") {
    rotationMode=BMM_Additive
    rotationSpace=BCS_BoneSpace
    scaleMode=BMM_Additive
    scaleSpace=BCS_BoneSpace
    translationMode=BMM_Additive
    translationSpace=BCS_BoneSpace
    alphaBoolBlend {blendOption=Linear}
    boneToModify {boneName=dyn_jewelry_B03}
   }
   n36 = "Transform (Modify) Bone - Bone: dyn_jewelry_B04"(ComponentPose=@n35, Alpha=1, Translation="0,0,0", Rotation="0,0,0", Scale="1,1,1") {
    rotationMode=BMM_Additive
    rotationSpace=BCS_BoneSpace
    scaleMode=BMM_Additive
    scaleSpace=BCS_BoneSpace
    translationMode=BMM_Additive
    translationSpace=BCS_BoneSpace
    alphaBoolBlend {blendOption=Linear}
    boneToModify {boneName=dyn_jewelry_B04}
   }
   n37 = "Transform (Modify) Bone - Bone: dyn_jewelry_B05"(ComponentPose=@n36, Alpha=1, Translation="0,0,0", Rotation="5,0,0", Scale="1,1,1") {
    rotationMode=BMM_Additive
    rotationSpace=BCS_BoneSpace
    scaleMode=BMM_Additive
    scaleSpace=BCS_BoneSpace
    translationMode=BMM_Additive
    translationSpace=BCS_BoneSpace
    alphaBoolBlend {blendOption=Linear}
    boneToModify {boneName=dyn_jewelry_B05}
    rotation {pitch=5}
   }
   n38 = "Transform (Modify) Bone - Bone: dyn_jewelry_Ba01"(ComponentPose=@n37, Alpha=1, Translation="0,0,0", Rotation="5,0,0", Scale="1,1,1") {
    rotationMode=BMM_Additive
    rotationSpace=BCS_BoneSpace
    scaleMode=BMM_Additive
    scaleSpace=BCS_BoneSpace
    translationMode=BMM_Additive
    translationSpace=BCS_BoneSpace
    alphaBoolBlend {blendOption=Linear}
    boneToModify {boneName=dyn_jewelry_Ba01}
    rotation {pitch=5}
   }
   n39 = "Transform (Modify) Bone - Bone: dyn_jewelry_Ba02"(ComponentPose=@n38, Alpha=1, Translation="0,0,0", Rotation="5,0,0", Scale="1,1,1") {
    rotationMode=BMM_Additive
    rotationSpace=BCS_BoneSpace
    scaleMode=BMM_Additive
    scaleSpace=BCS_BoneSpace
    translationMode=BMM_Additive
    translationSpace=BCS_BoneSpace
    alphaBoolBlend {blendOption=Linear}
    boneToModify {boneName=dyn_jewelry_Ba02}
    rotation {pitch=5}
   }
   n40 = "Transform (Modify) Bone - Bone: dyn_jewelry_Ba03"(ComponentPose=@n39, Alpha=1, Translation="0,0,0", Rotation="5,0,0", Scale="1,1,1") {
    rotationMode=BMM_Additive
    rotationSpace=BCS_BoneSpace
    scaleMode=BMM_Additive
    scaleSpace=BCS_BoneSpace
    translationMode=BMM_Additive
    translationSpace=BCS_BoneSpace
    alphaBoolBlend {blendOption=Linear}
    boneToModify {boneName=dyn_jewelry_Ba03}
    rotation {pitch=5}
   }
   n41 = "Transform (Modify) Bone - Bone: dyn_jewelry_C01"(ComponentPose=@n40, Alpha=1, Translation="0,0,0", Rotation="10,0,0", Scale="1,1,1") {
    rotationMode=BMM_Additive
    rotationSpace=BCS_BoneSpace
    scaleMode=BMM_Additive
    scaleSpace=BCS_BoneSpace
    translationMode=BMM_Additive
    translationSpace=BCS_BoneSpace
    alphaBoolBlend {blendOption=Linear}
    boneToModify {boneName=dyn_jewelry_C01}
    rotation {pitch=10}
   }
   n42 = "Transform (Modify) Bone - Bone: dyn_jewelry_C02"(ComponentPose=@n41, Alpha=1, Translation="0,0,0", Rotation="10,0,0", Scale="1,1,1") {
    rotationMode=BMM_Additive
    rotationSpace=BCS_BoneSpace
    scaleMode=BMM_Additive
    scaleSpace=BCS_BoneSpace
    translationMode=BMM_Additive
    translationSpace=BCS_BoneSpace
    alphaBoolBlend {blendOption=Linear}
    boneToModify {boneName=dyn_jewelry_C02}
    rotation {pitch=10}
   }
   n43 = Anim Dynamics - Bone: dyn_jewelry_A01!(ComponentPose=@n42, Alpha=1, GravityOverride:Vector="0,0,-980", ExternalForce="0,0,0") -> (Pose) {
    angularDampingOverride=0.999
    angularSpringConstant=200
    bGravityOverrideInSimSpace=true
    bOverrideAngularDamping=true
    bUseGravityOverride=true
    bUseSphericalLimits=true
    simulationSpace=BoneRelative
    alphaBoolBlend {blendOption=Linear}
    boundBone {boneName=dyn_jewelry_A01}
    componentAppliedLinearAccClamp=<2500 2500 1500>
    componentLinearAccScale=<0.5 0.5 0.3>
    componentLinearVelScale=<0.55 0.55 0.4>
    gravityOverride {z=500}
    physicsBodyDefinitions [{
      collisionType=CustomSphere
      sphereCollisionRadius=1
      boundBone {boneName=dyn_jewelry_A01}
      boxExtents=<20 20 20>
      constraintSetup {angularLimitsMax=<0 15 20>, angularLimitsMin=<0 -15 -5>, angularTarget=<0 0 0>}
      localJointOffset=<-5 -3 -0.1>
     }]
    relativeSpaceBone {boneName=gravity}
    sphericalLimits [{limitRadius=14, drivingBone {boneName=pelvis}, sphereLocalOffset=<-5 0 0>}]
   }
   n44 = Anim Dynamics - Bone: dyn_jewelry_A02!(ComponentPose=@n43, Alpha=1, GravityOverride:Vector="0,0,-980", ExternalForce="0,0,0") -> (Pose) {
    angularDampingOverride=0.999
    angularSpringConstant=200
    bGravityOverrideInSimSpace=true
    bOverrideAngularDamping=true
    bUseGravityOverride=true
    bUseSphericalLimits=true
    simulationSpace=BoneRelative
    alphaBoolBlend {blendOption=Linear}
    boundBone {boneName=dyn_jewelry_A02}
    componentAppliedLinearAccClamp=<2500 2500 1500>
    componentLinearAccScale=<0.5 0.5 0.3>
    componentLinearVelScale=<0.55 0.55 0.4>
    gravityOverride {z=500}
    physicsBodyDefinitions [{
      collisionType=CustomSphere
      sphereCollisionRadius=1
      boundBone {boneName=dyn_jewelry_A02}
      boxExtents=<20 20 20>
      constraintSetup {angularLimitsMax=<0 15 20>, angularLimitsMin=<0 -15 -5>, angularTarget=<0 0 0>}
      localJointOffset=<-5 -3 -0.1>
     }]
    relativeSpaceBone {boneName=gravity}
    sphericalLimits [{limitRadius=14, drivingBone {boneName=pelvis}, sphereLocalOffset=<-5 0 0>}]
   }
   n45 = Anim Dynamics - Bone: dyn_jewelry_A03!(ComponentPose=@n44, Alpha=1, GravityOverride:Vector="0,0,-980", ExternalForce="0,0,0") -> (Pose) {
    angularDampingOverride=0.999
    angularSpringConstant=200
    bGravityOverrideInSimSpace=true
    bOverrideAngularDamping=true
    bUseGravityOverride=true
    bUseSphericalLimits=true
    simulationSpace=BoneRelative
    alphaBoolBlend {blendOption=Linear}
    boundBone {boneName=dyn_jewelry_A03}
    componentAppliedLinearAccClamp=<2500 2500 1500>
    componentLinearAccScale=<0.5 0.5 0.3>
    componentLinearVelScale=<0.55 0.55 0.4>
    gravityOverride {z=500}
    physicsBodyDefinitions [{
      collisionType=CustomSphere
      sphereCollisionRadius=1
      boundBone {boneName=dyn_jewelry_A03}
      boxExtents=<15 15 15>
      constraintSetup {angularLimitsMax=<0 15 20>, angularLimitsMin=<0 -15 -5>, angularTarget=<0 0 0>}
      localJointOffset=<-5 -3 -0.1>
     }]
    relativeSpaceBone {boneName=gravity}
    sphericalLimits [{limitRadius=14, drivingBone {boneName=pelvis}, sphereLocalOffset=<-5 0 0>}]
   }
   n46 = Anim Dynamics - Bone: dyn_jewelry_B01!(ComponentPose=@n45, Alpha=1, GravityOverride:Vector="0,0,-980", ExternalForce="0,0,0") -> (Pose) {
    angularDampingOverride=0.999
    angularSpringConstant=200
    bGravityOverrideInSimSpace=true
    bOverrideAngularDamping=true
    bUseGravityOverride=true
    bUseSphericalLimits=true
    gravityScale=0.3
    simulationSpace=BoneRelative
    alphaBoolBlend {blendOption=Linear}
    boundBone {boneName=dyn_jewelry_B01}
    componentAppliedLinearAccClamp=<2500 2500 1500>
    componentLinearAccScale=<0.5 0.5 0.4>
    componentLinearVelScale=<0.55 0.55 0.45>
    gravityOverride {z=500}
    physicsBodyDefinitions [{
      collisionType=CustomSphere
      sphereCollisionRadius=1
      boundBone {boneName=dyn_jewelry_B01}
      boxExtents=<20 20 20>
      constraintSetup {angularLimitsMax=<0 2 5>, angularLimitsMin=<0 -15 0>, angularTarget=<0 0 0>}
      localJointOffset=<-3 -0.1 -0.1>
     }]
    relativeSpaceBone {boneName=gravity}
    sphericalLimits [{limitRadius=9.5, drivingBone {boneName=thigh_r}, sphereLocalOffset=<5 0 -4>}]
   }
   n47 = Anim Dynamics - Bone: dyn_jewelry_B02!(ComponentPose=@n46, Alpha=1, GravityOverride:Vector="0,0,-980", ExternalForce="0,0,0") -> (Pose) {
    angularDampingOverride=0.999
    angularSpringConstant=200
    bGravityOverrideInSimSpace=true
    bOverrideAngularDamping=true
    bUseGravityOverride=true
    bUseSphericalLimits=true
    gravityScale=0.3
    simulationSpace=BoneRelative
    alphaBoolBlend {blendOption=Linear}
    boundBone {boneName=dyn_jewelry_B02}
    componentAppliedLinearAccClamp=<2500 2500 1500>
    componentLinearAccScale=<0.5 0.5 0.4>
    componentLinearVelScale=<0.55 0.55 0.45>
    gravityOverride {z=500}
    physicsBodyDefinitions [{
      collisionType=CustomSphere
      sphereCollisionRadius=1
      boundBone {boneName=dyn_jewelry_B02}
      boxExtents=<20 20 20>
      constraintSetup {angularLimitsMax=<0 2 5>, angularLimitsMin=<0 -15 0>, angularTarget=<0 0 0>}
      localJointOffset=<-3 -0.1 -0.1>
     }]
    relativeSpaceBone {boneName=gravity}
    sphericalLimits [{limitRadius=9, drivingBone {boneName=thigh_r}, sphereLocalOffset=<5 0 -4>}]
   }
   n48 = Anim Dynamics - Bone: dyn_jewelry_B03!(ComponentPose=@n47, Alpha=1, GravityOverride:Vector="0,0,-980", ExternalForce="0,0,0") -> (Pose) {
    angularDampingOverride=0.999
    angularSpringConstant=200
    bGravityOverrideInSimSpace=true
    bOverrideAngularDamping=true
    bUseGravityOverride=true
    bUseSphericalLimits=true
    gravityScale=0.3
    simulationSpace=BoneRelative
    alphaBoolBlend {blendOption=Linear}
    boundBone {boneName=dyn_jewelry_B03}
    componentAppliedLinearAccClamp=<2500 2500 1500>
    componentLinearAccScale=<0.5 0.5 0.4>
    componentLinearVelScale=<0.55 0.55 0.45>
    gravityOverride {z=500}
    physicsBodyDefinitions [{
      collisionType=CustomSphere
      sphereCollisionRadius=1
      boundBone {boneName=dyn_jewelry_B03}
      boxExtents=<15 15 15>
      constraintSetup {angularLimitsMax=<0 2 5>, angularLimitsMin=<0 -15 0>, angularTarget=<0 0 0>}
      localJointOffset=<-3 -0.1 -1>
     }]
    relativeSpaceBone {boneName=gravity}
    sphericalLimits [{limitRadius=10.5, drivingBone {boneName=thigh_r}, sphereLocalOffset=<8 -1 -5>}]
   }
   n49 = Anim Dynamics - Bone: dyn_jewelry_B04!(ComponentPose=@n48, Alpha=1, GravityOverride:Vector="0,0,-980", ExternalForce="0,0,0") -> (Pose) {
    angularDampingOverride=0.999
    angularSpringConstant=200
    bGravityOverrideInSimSpace=true
    bOverrideAngularDamping=true
    bUseGravityOverride=true
    bUseSphericalLimits=true
    gravityScale=0.3
    simulationSpace=BoneRelative
    alphaBoolBlend {blendOption=Linear}
    boundBone {boneName=dyn_jewelry_B04}
    componentAppliedLinearAccClamp=<2500 2500 1500>
    componentLinearAccScale=<0.5 0.5 0.4>
    componentLinearVelScale=<0.55 0.55 0.45>
    gravityOverride {z=500}
    physicsBodyDefinitions [{
      collisionType=CustomSphere
      sphereCollisionRadius=1
      boundBone {boneName=dyn_jewelry_B04}
      boxExtents=<10 10 10>
      constraintSetup {angularLimitsMax=<0 2 5>, angularLimitsMin=<0 -15 0>, angularTarget=<0 0 0>}
      localJointOffset=<-3 -0.1 -3>
     }]
    relativeSpaceBone {boneName=gravity}
    sphericalLimits [{limitRadius=10, drivingBone {boneName=thigh_r}, sphereLocalOffset=<10 -1 -2>}]
   }
   n50 = Anim Dynamics - Bone: dyn_jewelry_B05!(ComponentPose=@n49, Alpha=1, GravityOverride:Vector="0,0,-980", ExternalForce="0,0,0") -> (Pose) {
    angularDampingOverride=0.999
    angularSpringConstant=200
    bGravityOverrideInSimSpace=true
    bOverrideAngularDamping=true
    bUseGravityOverride=true
    gravityScale=0.3
    simulationSpace=BoneRelative
    alphaBoolBlend {blendOption=Linear}
    boundBone {boneName=dyn_jewelry_B05}
    componentAppliedLinearAccClamp=<2500 2500 1500>
    componentLinearAccScale=<0.5 0.5 0.4>
    componentLinearVelScale=<0.55 0.55 0.2>
    gravityOverride {z=500}
    physicsBodyDefinitions [{
      collisionType=CustomSphere
      sphereCollisionRadius=1
      boundBone {boneName=dyn_jewelry_B05}
      boxExtents=<5 5 5>
      constraintSetup {angularLimitsMax=<0 2 5>, angularLimitsMin=<0 -15 0>, angularTarget=<0 0 0>}
      localJointOffset=<-3 -0.1 -1>
     }]
    relativeSpaceBone {boneName=gravity}
    sphericalLimits [{limitRadius=9.5, drivingBone {boneName=thigh_r}, sphereLocalOffset=<14 -1 0>}]
   }
   n51 = Anim Dynamics - Bone: dyn_jewelry_Ba01!(ComponentPose=@n50, Alpha=1, GravityOverride:Vector="0,0,-980", ExternalForce="0,0,0") -> (Pose) {
    angularDampingOverride=0.999
    angularSpringConstant=200
    bGravityOverrideInSimSpace=true
    bOverrideAngularDamping=true
    bUseGravityOverride=true
    bUseSphericalLimits=true
    gravityScale=0.3
    simulationSpace=BoneRelative
    alphaBoolBlend {blendOption=Linear}
    boundBone {boneName=dyn_jewelry_Ba01}
    componentAppliedLinearAccClamp=<2500 2500 1500>
    componentLinearAccScale=<0.5 0.5 0.4>
    componentLinearVelScale=<0.55 0.55 0.45>
    gravityOverride {z=500}
    physicsBodyDefinitions [{
      collisionType=CustomSphere
      sphereCollisionRadius=1
      boundBone {boneName=dyn_jewelry_Ba01}
      boxExtents=<20 20 20>
      constraintSetup {angularLimitsMax=<5 8 10>, angularLimitsMin=<-5 -15 -7>, angularTarget=<0 0 0>}
      localJointOffset=<-3 0.2 -2>
     }]
    relativeSpaceBone {boneName=gravity}
    sphericalLimits [{limitRadius=9, drivingBone {boneName=thigh_r}, sphereLocalOffset=<6 0 -4>}]
   }
   n52 = Anim Dynamics - Bone: dyn_jewelry_Ba02!(ComponentPose=@n51, Alpha=1, GravityOverride:Vector="0,0,-980", ExternalForce="0,0,0") -> (Pose) {
    angularDampingOverride=0.999
    angularSpringConstant=200
    bGravityOverrideInSimSpace=true
    bOverrideAngularDamping=true
    bUseGravityOverride=true
    bUseSphericalLimits=true
    gravityScale=0.3
    simulationSpace=BoneRelative
    alphaBoolBlend {blendOption=Linear}
    boundBone {boneName=dyn_jewelry_Ba02}
    componentAppliedLinearAccClamp=<2500 2500 1500>
    componentLinearAccScale=<0.5 0.5 0.4>
    componentLinearVelScale=<0.55 0.55 0.45>
    gravityOverride {z=500}
    physicsBodyDefinitions [{
      collisionType=CustomSphere
      sphereCollisionRadius=1
      boundBone {boneName=dyn_jewelry_Ba02}
      boxExtents=<20 20 20>
      constraintSetup {angularLimitsMax=<5 8 10>, angularLimitsMin=<-5 -15 -7>, angularTarget=<0 0 0>}
      localJointOffset=<-3 0.2 -2>
     }]
    relativeSpaceBone {boneName=gravity}
    sphericalLimits [{limitRadius=9, drivingBone {boneName=thigh_r}, sphereLocalOffset=<8 0 -4>}]
   }
   n53 = Anim Dynamics - Bone: dyn_jewelry_Ba03!(ComponentPose=@n52, Alpha=1, GravityOverride:Vector="0,0,-980", ExternalForce="0,0,0") -> (Pose) {
    angularDampingOverride=0.999
    angularSpringConstant=200
    bGravityOverrideInSimSpace=true
    bOverrideAngularDamping=true
    bUseGravityOverride=true
    bUseSphericalLimits=true
    gravityScale=0.3
    simulationSpace=BoneRelative
    alphaBoolBlend {blendOption=Linear}
    boundBone {boneName=dyn_jewelry_Ba03}
    componentAppliedLinearAccClamp=<2500 2500 1500>
    componentLinearAccScale=<0.5 0.5 0.4>
    componentLinearVelScale=<0.55 0.55 0.45>
    gravityOverride {z=500}
    physicsBodyDefinitions [{
      collisionType=CustomSphere
      sphereCollisionRadius=1
      boundBone {boneName=dyn_jewelry_Ba03}
      boxExtents=<15 15 15>
      constraintSetup {angularLimitsMax=<5 8 10>, angularLimitsMin=<-5 -15 -7>, angularTarget=<0 0 0>}
      localJointOffset=<-3 0.2 -2>
     }]
    relativeSpaceBone {boneName=gravity}
    sphericalLimits [{limitRadius=9.5, drivingBone {boneName=thigh_r}, sphereLocalOffset=<8 0 -1>}]
   }
   n54 = Anim Dynamics - Bone: dyn_jewelry_C01!(ComponentPose=@n53, Alpha=1, GravityOverride:Vector="0,0,-980", ExternalForce="0,0,0") -> (Pose) {
    angularDampingOverride=0.999
    angularSpringConstant=200
    bGravityOverrideInSimSpace=true
    bOverrideAngularDamping=true
    bUseGravityOverride=true
    bUseSphericalLimits=true
    gravityScale=0.3
    simulationSpace=BoneRelative
    alphaBoolBlend {blendOption=Linear}
    boundBone {boneName=dyn_jewelry_C01}
    componentAppliedLinearAccClamp=<2500 2500 1500>
    componentLinearAccScale=<0.5 0.5 0.3>
    componentLinearVelScale=<0.55 0.55 0.4>
    gravityOverride {z=500}
    physicsBodyDefinitions [{
      collisionType=CustomSphere
      sphereCollisionRadius=1
      boundBone {boneName=dyn_jewelry_C01}
      boxExtents=<20 20 20>
      constraintSetup {angularLimitsMax=<0 10 20>, angularLimitsMin=<0 -25 -10>, angularTarget=<0 0 0>}
      localJointOffset=<-4 0.2 -2>
     }]
    relativeSpaceBone {boneName=gravity}
    sphericalLimits [{limitRadius=9, drivingBone {boneName=thigh_r}, sphereLocalOffset=<1.8 0 -3>}]
   }
   n55 = Anim Dynamics - Bone: dyn_jewelry_C02!(ComponentPose=@n54, Alpha=1, GravityOverride:Vector="0,0,-980", ExternalForce="0,0,0") -> (Pose) {
    angularDampingOverride=0.999
    angularSpringConstant=200
    bGravityOverrideInSimSpace=true
    bOverrideAngularDamping=true
    bUseGravityOverride=true
    bUseSphericalLimits=true
    gravityScale=0.3
    simulationSpace=BoneRelative
    alphaBoolBlend {blendOption=Linear}
    boundBone {boneName=dyn_jewelry_C02}
    componentAppliedLinearAccClamp=<2500 2500 1500>
    componentLinearAccScale=<0.5 0.5 0.3>
    componentLinearVelScale=<0.55 0.55 0.4>
    gravityOverride {z=500}
    physicsBodyDefinitions [{
      collisionType=CustomSphere
      sphereCollisionRadius=1
      boundBone {boneName=dyn_jewelry_C02}
      boxExtents=<20 20 20>
      constraintSetup {angularLimitsMax=<0 10 20>, angularLimitsMin=<0 -25 -10>, angularTarget=<0 0 0>}
      localJointOffset=<-5 0.2 -2>
     }]
    relativeSpaceBone {boneName=gravity}
    sphericalLimits [{limitRadius=10, drivingBone {boneName=thigh_r}, sphereLocalOffset=<6 0 -2>}]
   }
   n56 = "Constraint - dyn_jewelry_B05 (PP)"!(ComponentPose=@n55, Alpha=1, ConstraintWeights_0=0.7, ConstraintWeights_1:float=0.3) -> (Pose) {
    alphaBoolBlend {blendOption=Linear}
    boneToModify {boneName=dyn_jewelry_B05}
    constraintSetup [{targetBone {boneName=dyn_jewelry_B05}}, {targetBone {boneName=dyn_jewelry_Ba03}}]
    constraintWeights [1,1]
   }
   Save cached pose 'BodyDynPose'(Pose=Component To Local(ComponentPose=@n56)) {cacheName=BodyDynPose}
  ]}
 EventGraph {nodes [
   n1 = Event BlueprintUpdateAnimation()
   > n2 = Sequence!(execute) -> (then_0, then_1, then_2:Exec, then_3:Exec)
   n3 = Sequence!(execute=@n2.then_1) -> (then_0, then_1, then_2)
   > Is Valid(InputObject=Get PlayerPawn() -> (PlayerPawn:Fort Player Pawn Athena Object Reference))
   > n4 = Branch(Condition=Get bIsInFrontEnd())
   Set WPO Globals Function(
    execute=@n4.else
    Material=Get Mat Instance Dynamic()
    Overlay Material=Get Overlay Mat Instance Dynamic()
    Skeletal Mesh=Get SkelMesh()
    "Local Noise Transform  Location"=Get Local Noise Transform Location() -> (Local Noise Transform Location:Vector)
    Local Noise Offset=Get Local Noise Offset()
   )
   > Set WPO Regional Function!(
    execute
    Material:Material Instance Dynamic Object Reference=Get Mat Instance Dynamic()
    Overlay Material:Material Instance Dynamic Object Reference=Get Overlay Mat Instance Dynamic()
    Skeletal Mesh:Skeletal Mesh Component Object Reference=Get SkelMesh()
    Velocity Transform:Name=foot_r
    Velocity Interp Speed:double=Get Velocity Interp Speed()
    Collision Transform:Name=calf_r
    Collision Transform Axis:EAxis Enum
    Invert Collision Transform Axis:bool=false
    Delta Time X:double=@n1.DeltaTimeX
    Previous Velocity Transform Location:"Vector (by ref)"=Get R1 Velocity Transform Location()
    Previous Velocity Transform Velocity:"Vector (by ref)"=Get R1 Velocity Transform Velocity()
    Region:int=1
   ) -> (then)
   > Set WPO Regional Function(
    Material=Get Mat Instance Dynamic()
    Overlay Material=Get Overlay Mat Instance Dynamic()
    Skeletal Mesh=Get SkelMesh()
    Velocity Transform=foot_l
    Velocity Interp Speed=Get Velocity Interp Speed()
    Collision Transform=calf_l
    Invert Collision Transform Axis=true
    Delta Time X=@n1.DeltaTimeX
    Previous Velocity Transform Location=Get R2 Velocity Transform Location()
    Previous Velocity Transform Velocity=Get R2 Velocity Transform Velocity()
    Region=2
   )
   > n5 = Branch(Condition=Get bIsSkydiveFloating())
   > n6 = Set WPO Pho Wind Intensity(WPO Pho Wind Intensity=0.5)
   SetScalarParameterValue(execute=[@n7.then, @n6.then], self=Get Mat Instance Dynamic(), ParameterName=ClothSim_WindStrength, Value=Get WPO Pho Wind Intensity())
   n7 = Set WPO Pho Wind Intensity(execute=@n5.else, WPO Pho Wind Intensity=0)
   n8 = Branch(execute=@n3.then_1, Condition=AND Boolean(A=Get bIsParachuteOpened(), B=Get custom_cape_equipped() -> (custom_cape_equipped:bool)))
   > Do Once(Reset=@n8.else)
   > n9 = Cast To FortPlayerPawn(Object=TryGetFortPawn())
   > Is Valid(InputObject=@n9)
   > n10 = Cast To B_Athena_PartModifier_ArcticIce_FaceAcc!(
    execute
    Object=GetModifierForPartType!(self:Fort Player Pawn Object Reference=@n9, PartType:EFortCustomPartType Enum=Face) -> (ReturnValue:Custom Character Part Modifier Object Reference)
   ) -> (then, CastFailed, AsB Athena Part Modifier Arctic Ice Face Acc:B Athena Part Modifier Arctic Ice Face Acc Object Reference) {is=title K2Node_DynamicCast}
   > Set Scarf Visible(Scarf Visible:bool=Get Scarf Visible!(self:B Athena Part Modifier Arctic Ice Face Acc Object Reference=@n10) -> (Scarf Visible:bool)) -> (Output_Get:bool)
   Set pantlegAlphaFrontend(execute=@n3.then_2, pantlegAlphaFrontend:double=SelectFloat(A=0.65, B=1, bPickA=Get bIsInFrontEnd())) -> (Output_Get:double)
   n11 = Gravity Override State Function DEPRECATED(
    execute=@n2.then_0
    CharacterPartAnimInstance=Self-Reference()
    GameTimeInSeconds=GetGameTimeInSeconds()
    GravityOverrideParams=Get Gravity Override Params()
   )
   > Set Gravity Override Cape(Gravity Override Cape:Vector=@n11) -> (Output_Get:Vector)
   Event BlueprintInitializeAnimation()
   > Set SkelMesh(SkelMesh=GetOwnerSkeletalMeshComponent())
   n12 = TryGetPawnOwner()
   > Is Valid(InputObject=@n12)
   > n13 = Cast To FortPlayerPawnAthena(Object=@n12)
   > Set PlayerPawn(PlayerPawn:Fort Player Pawn Athena Object Reference=@n13) -> (Output_Get:Fort Player Pawn Athena Object Reference)
   Event OnParachuteClosed()
   > n14 = Branch(Condition=Get custom_cape_equipped())
   > Set Scarf Visible(Scarf Visible=true)
   Set Scarf Visible(execute=@n14.else, Scarf Visible=false)
   Event OnCharacterCustomizationLoaded()
   > n15 = Sequence()
   > n16 = Set SkelMesh(SkelMesh=GetPartSkeletalMeshComponent())
   > n17 = Get Dynamic Material Instance(SkeletalMeshComponent=@n16)
   > Set Mat Instance Dynamic(Mat Instance Dynamic=@n17)
   > n18 = GetOrCreateDynamicOverlayMaterial(MeshComponent=@n16, MaterialName=None)
   > Set Overlay Mat Instance Dynamic(Overlay Mat Instance Dynamic=@n18)
   n19 = CheckForBackbling()
   Set isWearingPackBool(execute=@n15.then_2, isWearingPackBool:bool=OR Boolean!(A=@n19, B=@n19.bIsBackblingACape, C:bool=@n19.bIsBackblingLong) -> (ReturnValue)) -> (Output_Get:bool)
   n20 = Branch(execute=@n15.then_1, Condition=CheckForBackbling().bIsBackblingACape)
   > n21 = Branch(Condition=BackpackItemDefinitionHasAnyMetaTag!(Tags:Array of Gameplay Tag Structures=Make Array("[0]"=(TagName="Cosmetics.MetaTags.ArcticIce.Cape")), bRequiresEquipped:bool=false) -> (ReturnValue:bool))
   > Set custom_cape_equipped(custom_cape_equipped:bool=true) -> (Output_Get:bool)
   > Set Scarf Visible(Scarf Visible=true)
   Set custom_cape_equipped(execute=@n21.else, custom_cape_equipped=false)
   > Set Scarf Visible(Scarf Visible=false)
   Set custom_cape_equipped(execute=@n20.else, custom_cape_equipped=false)
   > Set Scarf Visible(Scarf Visible=false)
   AND Boolean(A=@n19, B=NOT Boolean(A=@n19.bIsBackblingLong))
  ]}
]
variables [
 custom_cape_equipped {type=boolean}
 Gravity Override Cape {type=Vector}
 Gravity Override Params {
  type=GravityOverrideParamsStruct
  default {
   riding_Moving_Targeting {gravityOverride=<0 -200 -800>}
   riding_Standing {gravityOverride=<0 -200 -980>}
   skydive_DivingDown {gravityOverride=<0 0 500>, windAmplitude=<0 0 0>, windFrequency=<0 0 0>}
   skydive_FloatingDown {jointName=pelvis, gravityOverride=<0 0 500>, windAmplitude=<0 0 0>, windFrequency=<0 0 0>, windOffset=<0 0 0>}
   skydive_GoingUp {gravityOverride=<0 0 500>, windAmplitude=<50 -100 400>, windOffset=<-50 100 0>}
   skydive_Parachuting {gravityOverride=<0 0 -980>, windAmplitude=<100 -100 400>}
  }
 }
 isWearingPackBool {type=boolean}
 Local Noise Offset {type=Vector}
 Local Noise Transform Location {type=Vector}
 Mat Instance Dynamic {type=MaterialInstanceDynamic}
 Overlay Mat Instance Dynamic {type=MaterialInstanceDynamic}
 pantlegAlphaFrontend {type=number}
 PlayerPawn {type=FortPlayerPawnAthena}
 R1 Velocity Transform Location {type=Vector}
 R1 Velocity Transform Velocity {type=Vector}
 R2 Velocity Transform Location {type=Vector}
 R2 Velocity Transform Velocity {type=Vector}
 Scarf Visible {type=boolean}
 SkelMesh {type=SkeletalMeshComponent}
 Velocity Interp Speed {type=number, default=30}
 WPO Pho Wind Intensity {type=number}
]
